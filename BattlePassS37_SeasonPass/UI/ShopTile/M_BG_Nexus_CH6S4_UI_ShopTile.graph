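# Translucent UI material for the Chapter 6 Season 4 Nexus item shop tile background, with an animated backdrop and character texture.
# A radial two-color base, two panning cloud layers with Screen blends, counter-panning funnel-UV light rays and gradient-masked confetti embers form named reroute layers.
# Fire glow, faint noise, a positionable character texture and a rounded-corner function are also present.
# Ray noise samples use explicit UV derivatives for mip selection.
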Material /BattlePassS37_SeasonPass/UI/ShopTile/M_BG_Nexus_CH6S4_UI_ShopTile
graph {
 properties {blendMode=BLEND_Translucent, materialDomain=MD_UI, pixelDepthOffsetMode=PDOM_Legacy, usedShadingModels=Default Lit}
 nodes [
  n1 = Constant()
  n2 = Time()
  n3 = ScalarParameter() {defaultValue=0.01, group=LightRays, parameterName=LightRaysSpeed}
  AppendVector(A=@n1, B=Frac(None=Multiply(A=@n2, B=@n3) {constB=-0.25}))
  n4 = Noise!(World Position, FilterWidth) {noiseFunction=NOISEFUNCTION_ValueALU}
  n5 = Add(
   A=NamedRerouteUsage()
   B=AppendVector(A=ScalarParameter() {group=FireGlow, parameterName=FireGlowX}, B=ScalarParameter() {group=FireGlow, parameterName=FireGlowY})
  )
  n6 = ScalarParameter() {defaultValue=1, group=FireGlow, parameterName=FireGlowScale}
  n7 = LinearInterpolate(
   A=NamedRerouteUsage()
   B=Constant3Vector() {constant=<rgba 1 0.1395 0.0614 1>}
   Alpha=MaterialFunctionCall(
    UV=MaterialFunctionCall(UVs=@n5, Texture Scale=@n6) {materialFunction=&ScaleUVsByCenter}.UVs
    GradientStart=Constant()
    GradientEnd=Constant() {r=0.5}
   ) {materialFunction=&MF_UI_RadialGradient}.Smooth
  )
  n8 = Constant3Vector() {constant=<rgba 0.9635 0.3867 0.0015 1>}
  n9 = NamedRerouteUsage()
  n10 = MaterialFunctionCall() {materialFunction=&MF_UI_UVScaling}
  n11 = ScalarParameter() {defaultValue=1, group=Character, parameterName=Texture Scale}
  n12 = ScalarParameter() {group=Character, parameterName=X Location}
  n13 = ScalarParameter() {group=Character, parameterName=Y Location}
  n14 = TextureSampleParameter2D(UVs=Add(
    A=MaterialFunctionCall(UVs=@n10.Scaled To Fill, Texture Scale=@n11) {materialFunction=&ScaleUVsByCenter}.UVs
    B=AppendVector(A=@n12, B=@n13)
   )) {group=Character, parameterName=Texture, texture=&/T-AthenaSoldiers-Character-VoidRedemption-Rebel}
  LinearInterpolate(A=Add(A=Multiply(A=@n4) {constB=0.011}, B=LinearInterpolate(A=@n7, B=@n8, Alpha=@n9)), B=@n14.RGB, Alpha=@n14.RGB)
  MaterialFunctionCall!(Input Alpha=Constant() {r=1}, bOverrideCornerRadius, Corner Radius Override) {materialFunction=/FortUILibrary/Systems/ItemShopTile/Materials/MaterialFunctions/MF_UI_ItemShop_RoundedCorners}
  n15 = ScalarParameter() {defaultValue=1, parameterName=L1-Patter-X-Offset}
  n16 = ScalarParameter() {defaultValue=0.172, parameterName=L1-Patter-Y-Offset}
  n17 = MaterialFunctionCall() {materialFunction=&GetUserInterfaceUV}
  n18 = MaterialFunctionCall!(
   "UV(V2)"=Add(A=NamedRerouteUsage(), B=AppendVector(A=@n15, B=@n16))
   Global-Angle=LinearInterpolate(Alpha=ComponentMask(None=@n17.Normalized UV) {g=true}) {constA=-1.05, constB=-1}
   "Pattern-Tiles(V2)"=AppendVector(A=ScalarParameter() {defaultValue=80, parameterName=L1-Pattern-X}, B=ScalarParameter() {defaultValue=0.7, parameterName=L1-Pattern-Y})
   RandomSeed=ScalarParameter() {defaultValue=12345, parameterName=L1-RandomSeed}
   Confetti-Speed-TimeFactor=ScalarParameter() {defaultValue=0.138, parameterName=L1-Confetti-Speed-TimeFactor}
   Confetti-Rotation-TimeFactor=ScalarParameter() {defaultValue=0.794, parameterName=L1-Confetti-Rotation-TimeFactor}
   Confetti-Local-Rotation-Range=ScalarParameter() {defaultValue=0.0075, parameterName=L1-Confetti-Local-Rotation-Range}
   "Confetti-Local-Tiles(V2)"=AppendVector(
    A=ScalarParameter() {defaultValue=5, parameterName=L1-Confetti-Local-Tile-X}
    B=ScalarParameter() {defaultValue=800, parameterName=L1-Confetti-Local-Tile-Y}
   )
   "SmoothStepRanges(V2)"=AppendVector(
    A=ScalarParameter() {defaultValue=0.65, parameterName=L1-SmoothStep-Max}
    B=ScalarParameter() {defaultValue=0.5, parameterName=L1-SmoothStep-Min}
   )
  ) {materialFunction=/Game/UI/Frontend/ItemShop/Art/Cactus/MF_UI_Confetti}
  n19 = ComponentMask(None=TextureCoordinate()) {g=true}
  n20 = ScalarParameter() {defaultValue=0.3, parameterName=L1-Gradient-Mask}
  n21 = Saturate(None=Multiply(A=Subtract(A=@n19, B=@n20), B=ScalarParameter() {defaultValue=5, parameterName=L1-Gradient-Falloff}))
  NamedRerouteDeclaration(None=Multiply(A=Multiply(A=@n18.Confetti-Mask, B=@n21), B=ScalarParameter() {defaultValue=1, parameterName=L1-Intensity})) {name=Embers_2, nodeColor=<rgba 0 0.4375 1 1>}
  n22 = MaterialFunctionCall() {materialFunction=&MF_UI_UVScaling}
  n23 = Constant()
  n24 = Constant() {r=0.175}
  NamedRerouteDeclaration(None=Add(A=@n22.Scaled to Fit, B=AppendVector(A=@n23, B=@n24))) {name=Focal Point UV, nodeColor=<rgba 0 0 0 1>}
  NamedRerouteDeclaration(None=@n22.Scaled To Fill) {name=ScaleToFitUV, nodeColor=<rgba 0 0 0 1>}
  n25 = VectorParameter() {parameterName=BackgroundLight, defaultValue=<rgba 0.0742 0.2423 0.6867 1>}
  n26 = NamedRerouteUsage()
  n27 = AppendVector(
   A=ScalarParameter() {defaultValue=-0.15, group=LightRays, parameterName=RadialX}
   B=ScalarParameter() {defaultValue=0.15, group=Background, parameterName=RadialY}
  )
  n28 = MaterialFunctionCall(
   UV=MaterialFunctionCall(UVs=Add(A=@n26, B=@n27), Texture Scale=ScalarParameter() {defaultValue=1.25, group=Background, parameterName=RadialSize}) {materialFunction=&ScaleUVsByCenter}.UVs
   GradientStart=ScalarParameter() {group=Background, parameterName=RadGradStart}
   GradientEnd=ScalarParameter() {defaultValue=1, group=Background, parameterName=RadGradEnd}
  ) {materialFunction=&MF_UI_RadialGradient}
  NamedRerouteDeclaration(None=LinearInterpolate(
    A=VectorParameter() {parameterName=BackgroundDark, defaultValue=<rgba 0.0048 0.0144 0.0284 1>}.RGB
    B=NamedRerouteDeclaration(None=@n25.RGB) {name=ColorBackgroundLight, nodeColor=<rgba 0 0 0 1>}
    Alpha=MaterialFunctionCall("Value (0-1)"=@n28.Smooth) {materialFunction=&MF_UI_EaseCurves}.EaseInOutSine
   )) {name=RadialBackgroundBase, nodeColor=<rgba 0 0 0 1>}
  NamedRerouteDeclaration(None=@n22.ScaleX) {name=FocalMask, nodeColor=<rgba 0 0 0 1>}
  n29 = NamedRerouteUsage()
  n30 = ScalarParameter() {defaultValue=1.2, group=MiddleClouds, parameterName=Mid_MCloudSize, sliderMax=10}
  n31 = Constant2Vector() {g=0.125, r=-0.125}
  n32 = ScalarParameter() {defaultValue=0.2, group=MiddleClouds, parameterName=Mid_MCloudSpeedMult, sliderMax=1}
  n33 = TextureSample(UVs=Panner(
    Coordinate=MaterialFunctionCall(UVs=@n29, Texture Scale=@n30) {materialFunction=&ScaleUVsByCenter}.UVs
    Speed=Multiply(A=@n31, B=@n32)
   ) {speedX=-0.125, speedY=0.125}) {texture=/BattlePassS37_SeasonPass/UI/Textures/Background/T_Clouds}
  NamedRerouteDeclaration(None=@n33.R) {name=M Cloud R Mid, nodeColor=<rgba 0 0 0 1>}
  NamedRerouteDeclaration(None=@n33.A) {name=M Cloud A Mid, nodeColor=<rgba 0 0 0 1>}
  n34 = NamedRerouteUsage()
  n35 = Multiply(
   A=Add(A=NamedRerouteUsage(), B=NamedRerouteUsage())
   B=ScalarParameter() {defaultValue=0.6, group=MiddleClouds, parameterName=Mid_MCloudDensityMult}
  )
  n36 = MaterialFunctionCall(UV=MaterialFunctionCall(UVs=NamedRerouteUsage(), Texture Scale=Constant() {r=3}) {materialFunction=&ScaleUVsByCenter}.UVs) {materialFunction=&MF_UI_RadialGradient}
  n37 = Abs(None=Subtract(A=Min(A=@n35, B=@n36.Smooth), B=@n36.Smooth))
  n38 = Multiply(
   A=VectorParameter() {group=MiddleClouds, parameterName=MidMColor, defaultValue=<rgba 0.004 0.0232 0.0762 1>}.RGB
   B=LinearInterpolate(
    A=NamedRerouteUsage()
    B=NamedRerouteDeclaration(None=@n37) {name=MidMCloudMask, nodeColor=<rgba 0 0 0 1>}
    Alpha=ScalarParameter() {defaultValue=0.4, group=MiddleClouds, parameterName=MidMCloudOpacity}
   ) {constAlpha=0.3}
  )
  n39 = Multiply(
   A=Add(A=NamedRerouteUsage(), B=NamedRerouteUsage())
   B=ScalarParameter() {defaultValue=0.6, group=MiddleClouds, parameterName=Bot_MCloudDensityMult}
  )
  n40 = MaterialFunctionCall(UV=MaterialFunctionCall(UVs=NamedRerouteUsage(), Texture Scale=Constant() {r=5}) {materialFunction=&ScaleUVsByCenter}.UVs) {materialFunction=&MF_UI_RadialGradient}
  n41 = Abs(None=Subtract(A=Min(A=@n39, B=@n40.Smooth), B=@n40.Smooth))
  n42 = Multiply(
   A=VectorParameter() {group=MiddleClouds, parameterName=BotMColor, defaultValue=<rgba 0.004 0.0232 0.0762 1>}.RGB
   B=LinearInterpolate(
    A=NamedRerouteUsage()
    B=NamedRerouteDeclaration(None=@n41) {name=BotMCloudMask, nodeColor=<rgba 0 0 0 1>}
    Alpha=ScalarParameter() {defaultValue=0.2, group=MiddleClouds, parameterName=BotMCloudOpacity}
   ) {constAlpha=0.2}
  )
  n43 = Add(A=@n38, B=Saturate(None=Add(A=@n38, B=@n42)))
  n44 = MaterialFunctionCall!(Blend=LinearInterpolate(B=Saturate(None=@n43)) {constAlpha=1}, Base=NamedRerouteUsage()) {materialFunction=/Engine/Functions/Engine_MaterialFunctions03/Blends/Blend_Screen}
  NamedRerouteDeclaration(None=Saturate(None=MaterialFunctionCall(Blend=@n34, Base=@n44.Result) {materialFunction=&Blend_Screen}.Result)) {name=MiddleCloud, nodeColor=<rgba 0 0 0 1>}
  n45 = NamedRerouteUsage()
  n46 = ScalarParameter() {defaultValue=1.2, group=MiddleClouds, parameterName=Bot_MCloudSize, sliderMax=10}
  n47 = Constant2Vector() {g=0.125, r=0.07}
  n48 = ScalarParameter() {defaultValue=0.1, group=MiddleClouds, parameterName=Bot_MCloudSpeedMult, sliderMax=1}
  n49 = TextureSample(UVs=Panner(
    Coordinate=MaterialFunctionCall(UVs=@n45, Texture Scale=@n46) {materialFunction=&ScaleUVsByCenter}.UVs
    Speed=Multiply(A=@n47, B=@n48)
   ) {speedX=-0.125, speedY=0.125}) {texture=&T_Clouds}
  NamedRerouteDeclaration(None=@n49.R) {name=M Cloud R Bot, nodeColor=<rgba 0 0 0 1>}
  NamedRerouteDeclaration(None=@n49.A) {name=M Cloud A Bot, nodeColor=<rgba 0 0 0 1>}
  n50 = NamedRerouteUsage()
  n51 = ScalarParameter() {defaultValue=1, group=LightRays, parameterName=FocalMaskMin}
  n52 = Multiply(A=ScalarParameter() {defaultValue=2, group=LightRays, parameterName=FocalMaskMax}) {constB=2}
  n53 = NamedRerouteUsage()
  n54 = NamedRerouteUsage()
  n55 = DDX(Value=@n54)
  n56 = DDY(Value=@n54)
  n57 = ScalarParameter() {defaultValue=-0.15, group=LightRays, parameterName=RaysX}
  n58 = ScalarParameter() {defaultValue=0.15, group=LightRays, parameterName=RaysY}
  n59 = MaterialFunctionCall!(
   UV=Add(A=NamedRerouteUsage(), B=AppendVector(A=@n57, B=@n58))
   Tiling=AppendVector(
    A=ScalarParameter() {defaultValue=3, group=LightRays, parameterName=LightRaysFrequency}
    B=ScalarParameter() {defaultValue=0.001, group=LightRays, parameterName=LightRaysStretch}
   )
   Offset=@n1
   Mask Size=Constant()
   Blend Distance=Constant() {r=1}
  ) {materialFunction=/FortUILibrary/MaterialFunctions/Effects/MF_UI_Funnel}
  n60 = Panner(Coordinate=@n59.Wrapped Funnel UV) {speedX=-0.01}
  n61 = TextureSample!(UVs=@n60, Tex, "DDX(UVs)"=@n55, "DDY(UVs)"=@n56, Apply View MipBias) {mipValueMode=TMVM_Derivative, texture=/Game/Global/Textures/Default/Noise/Noise}
  n62 = TextureSample(UVs=@n60, "DDX(UVs)"=@n55, "DDY(UVs)"=@n56) {mipValueMode=TMVM_Derivative, samplerType=SAMPLERTYPE_Grayscale, texture=&LowResBlurredNoise}
  n63 = Divide(
   A=NamedRerouteUsage()
   B=Multiply(A=LinearInterpolate(A=@n51, B=@n52, Alpha=@n53) {constA=1, constB=4}, B=LinearInterpolate(A=@n61.R, Alpha=@n62.R))
  )
  n64 = ScalarParameter() {defaultValue=1, group=LightRays, parameterName=FocalMaskMin}
  n65 = ScalarParameter() {defaultValue=2, group=LightRays, parameterName=FocalMaskMax}
  n66 = NamedRerouteUsage()
  n67 = Panner(Coordinate=@n59.Wrapped Funnel UV) {speedX=0.01}
  n68 = TextureSample(UVs=@n67, "DDX(UVs)"=@n55, "DDY(UVs)"=@n56) {mipValueMode=TMVM_Derivative, texture=&Noise}
  n69 = TextureSample(UVs=@n67, "DDX(UVs)"=@n55, "DDY(UVs)"=@n56) {mipValueMode=TMVM_Derivative, samplerType=SAMPLERTYPE_Grayscale, texture=&LowResBlurredNoise}
  n70 = Divide(
   A=NamedRerouteUsage()
   B=Multiply(A=LinearInterpolate(A=@n64, B=@n65, Alpha=@n66) {constA=1, constB=2}, B=LinearInterpolate(A=@n68.R, Alpha=@n69.R))
  )
  n71 = Saturate(None=Add(A=Multiply(A=@n63) {constB=0.1}, B=Multiply(A=@n70) {constB=0.1}))
  NamedRerouteDeclaration(None=LinearInterpolate(A=LinearInterpolate(A=@n50, B=@n71), B=NamedRerouteUsage(), Alpha=@n50)) {name=LightRays, nodeColor=<rgba 0 0 0 1>}
  ScalarParameter() {defaultValue=-1, parameterName=L1-Global Angle}
 ]
}
